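# NS_Magic_AOE_Base is a GPU Niagara system for area-wide magical ambience: colored dots, purple fireflies and low fog.
# Dots fill a wide box, flicker, and drift with wind scaled by the Fortnite gust parameter collection.
# Fireflies rise through curl noise, gust-scaled wind and distance-field avoidance, rendered as meshes and sprites.
# Fog sprites fade in a flattened, wind-blown layer; texture-mask and landscape placement in fog and fireflies is disabled.
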NiagaraSystem /SparksMaps/Effects/Environment/S7/Magic/NS_Magic_AOE_Base
dependencies {
 usedDataInterfaces [
  /Script/Niagara.NiagaraDataInterfaceTexture
  &NiagaraDataInterfaceLandscape
  &NiagaraDataInterfaceCurve
  &NiagaraDataInterfaceCollisionQuery
  &NiagaraDataInterfaceColorCurve
 ]
 usedDynamicInputs []
 usedModules [
  &self:SystemSpawnScript.SystemScriptSource.SystemScriptGraph.NiagaraNodeAssignment_1.SetVariables_#1
  &self:SystemSpawnScript.SystemScriptSource.SystemScriptGraph.NiagaraNodeAssignment_0.SetVariables_#2
  &SystemState
 ]
}
emitters {
 Dots_NoTilt {
  simTarget=GPUComputeSim
  emitterUpdateScript [
   &EmitterState(Life Cycle Mode=::NewEnumerator1, Loop Duration=2, UseLoopDelay=false)
   &SpawnRate(
    Spawn Probability=&FloatFromCurve(
     FloatCurve=DataInterface() {propertyValues {Curve {keys [<at 0 0 RCIM_Constant>,<at 0.25 0 RCIM_Constant>,<at 0.5 0 RCIM_Constant>,<at 0.75 1 RCIM_Constant>,<at 1 1 RCIM_Constant>,<at 1 0 RCIM_Constant>]}}}
     CurveIndex=&MakeFloatFromLinearColor(LinearColor=$NPC.FortniteNPC.FortniteActiveTimeOfDay, Channel:ENiagaraLinearColor_Channels=::NewEnumerator3)
    )
    SpawnRate=350
    Use Spawn Probability=false
   )
  ]
  particleSpawnScript [
   &InitializeParticle(
    Color=&ColorFromCurve(
     ColorCurve=DataInterface() {propertyValues {
       AlphaCurve {keys [<at 0.4738 0.0315 RCIM_Cubic>]}
       BlueCurve {keys [<at 0 0.7166>,<at 0.3901 1>,<at 0.7425 0.9051>,<at 1 1 RCIM_Cubic>]}
       GreenCurve {keys [<at 0 1>,<at 0.3901 0>,<at 0.7425 0>,<at 1 0.4423 RCIM_Cubic>]}
       RedCurve {keys [<at 0 0>,<at 0.3901 0.4491>,<at 0.7425 1>,<at 1 0.1248 RCIM_Cubic>]}
      }}
     CurveIndex=&RandomRangeFloat(Minimum=0, Maximum=1)
    )
    Lifetime Max=8
    Lifetime Min=5
    Lifetime Mode=::NewEnumerator1
    Sprite Rotation Mode=::NewEnumerator2
    Sprite Size Mode=::NewEnumerator1
    Sprite UV Mode=::NewEnumerator3
    Uniform Sprite Size Max=75
    UsePositionOffset=false
    Write Material Random=true
   )
   &ShapeLocation(
    Box Midpoint=&RandomRangeVector(Minimum=<0 0 0>, Maximum=<1 1 1>)
    Box Size=<6500 6500 400>
    MaskSpawnGroup=false
    Override Seed=false
    Shape Primitive=::NewEnumerator2
    Surface Only=false
    UseOwnerScale=true
   )
   &DynamicMaterialParameters(
    Index 0 Param 1=&RandomRangeFloat(Minimum=0.1, Maximum=12)
    Module.Index0Param0Write=true
    Module.Index0Param1Write=true
    Module.Index0Param2Write=true
    Module.Index0Param3Write=true
   )
  ]
  particleUpdateScript [
   &ParticleState()
   &WindForce(
    Frequency=6
    Offset By Time:Vector3f=<0 0 -50>
    Scale=1
    Wind Speed=<1 0 0>
    Wind Speed Scale=&Lerp_Float(A=0.25, B=150, Alpha=$NPC.FortniteNPC.WindGustsSystem_Active)
   )
   &AerodynamicDrag(
    Aerodynamic Drag:NiagaraFloat=&RandomRangeFloat(Minimum=0.6, Maximum=1)
    Aerodynamic Rotational Drag=0.2
    Drag Shape Curve=DataInterface() {propertyValues {Curve {keys [
        {interpMode=RCIM_Cubic, tangentMode=RCTM_User, time=0, value=0.1}
        {arriveTangent=2.0799, interpMode=RCIM_Cubic, leaveTangent=2.0799, tangentMode=RCTM_User, time=1, value=1}
       ]}}}
    Lift Shape Curve=DataInterface() {propertyValues {Curve {keys [
        {interpMode=RCIM_Cubic, tangentMode=RCTM_User, time=0, value=0}
        {interpMode=RCIM_Cubic, tangentMode=RCTM_User, time=0.5, value=1}
        {tangentMode=RCTM_User, time=1, value=0}
       ]}}}
    Use Linear Drag=true
    Use Rotational Drag=true
   )
   &ScaleSpriteSize(Scale Factor=&Vector2DFromFloat(Value=&FloatFromCurve(FloatCurve=DataInterface() {propertyValues {Curve {keys [
          <at 0 0 RCIM_Cubic>
          <at 0.112 1.0173>
          {arriveTangent=-1.142, interpMode=RCIM_Cubic, leaveTangent=-1.142, time=0.7001, value=0.872}
          <at 1.0011 0.002>
         ]}}}))) {enabled=false}
   &ScaleColor(
    Scale Alpha=&FloatFromCurve(
     FloatCurve=DataInterface() {propertyValues {Curve {keys [<at 0 0 RCIM_Cubic>,<at 0.5 1 RCIM_Cubic>,<at 1 0 RCIM_Cubic>]}}}
     Scale Curve=&Multiply_Float(
      A=0.75
      B=&Waveform(
       Global Input Position=&RandomRangeFloat(Minimum=0, Maximum=1)
       Global Clamp Min=0.5
       "Waveform [1]"=::NewEnumerator9
       "[1] Frequency"=5
       "[1] Amplitude Min/Max"=<0 1>
       "[1] Scale"=0.5
       "[1] Bias"=1
      )
     )
    )
    ScaleA=true
    ScaleRGB=true
   )
   &SolveForcesAndVelocity(Clamp Velocity=false, Limit Acceleration=false)
   &KillParticlesInVolume(
    Box Size:Vector3f=<3000 3000 1500>
    Kill Shape=::NewEnumerator1
    Volume Origin:Vector3f=&AddVectorToPosition(Position=<-74803 4342 2309.74>, Vector=<0 0 0>)
   ) {enabled=false}
   &KillParticlesInVolume(
    Box Size:Vector3f=<4500 4500 7500>
    Kill Shape=::NewEnumerator1
    Volume Origin:Vector3f=&AddVectorToPosition(Position=<-83738.4688 4360.5601 4476.9702>, Vector=<0 0 0>)
   ) {enabled=false, module=KillParticlesInVolume001}
  ]
  rendererClasses [&NiagaraSpriteRendererProperties]
  renderers [{rendererClass=&NiagaraSpriteRendererProperties}]
 }
 Firefiles_Epic {
  simTarget=GPUComputeSim
  emitterSpawnScript [/SetVariables(Emitter.Camera Position:NiagaraPosition=<0 0 -50>) {isSetParameters=true, module=SetVariables_#3}]
  emitterUpdateScript [
   &EmitterState()
   &SpawnRate(Spawn Probability=0.5, SpawnRate=30, Use Spawn Probability=true)
  ]
  particleSpawnScript [
   &InitializeParticle(
    Color=<rgba 0.5645 0.1094 1.5 0.8>
    Lifetime Max=8
    Lifetime Min=5
    Lifetime Mode=::NewEnumerator1
    Mesh Scale Mode=::NewEnumerator1
    Mesh Uniform Scale Max=0.6
    Mesh Uniform Scale Min=0.36
    Position=$Emitter.Camera Position
    Position Mode=::NewEnumerator1
    Sprite Size Mode=::NewEnumerator3
    Uniform Sprite Size=150
    UsePositionOffset=false
    Write Material Random=true
   )
   &ShapeLocation(MaskSpawnGroup=false, Non Uniform Scale=<1 1 0.1>, Override Seed=false, Sphere Radius=7000, UseOwnerScale=true)
   /Niagara/Modules/Update/Textures/SampleTexture(
    Apply Sampled RGBA to Particle Color=false
    Comparison Type:ENiagaraBooleanLogicOps_v2=::NewEnumerator6
    "Comparison Value (A)":ENiagara_ImportanceRejectionMode=::NewEnumerator4
    "Comparison Value (B)"=0.17
    Enable Sample Rejection=true
    EnableFalloff=false
    Generate UV Using Importance Sampling=true
    Rotation Coordinate Space:ENiagaraCoordinateSpace=::World
    Rotation Mode:ENiagara_RotationMode=::NewEnumerator0
    Texture=DataInterface() {propertyValues {
      Texture=/Game/Effects/Fort_Effects/Effects/Environment/MagicForest/Systems/Textures/T_MagicField_FXMask
      TextureUserParameter {parameter {name=None, typeDefHandle {registeredTypeIndex=2}, varData []}}
     }}
    Transform UV Into Position=true
    UV Origin:NiagaraPosition=&AddVectorToPosition(Position=$Emitter.Camera Position, Vector=<0 -8750 500>)
    UV Scale:Vector2f=<63500 63500>
   ) {enabled=false}
   &self:SampleLandscape() {enabled=false, module=SampleLandscape}
   &KillParticlesInVolume(Invert Volume=true, Sphere Radius=11000) {enabled=false}
   &CameraOffset(Camera Offset Amount=-100)
   /SetVariables(Particles.Spline Scale:Vector3f=<1 1 1>) {isSetParameters=true}
   &AccelerationForce(Acceleration=<0 0 500>) {module=AccelerationForce001}
   &ApplyInitialForces()
  ]
  particleUpdateScript [
   &ParticleState()
   &ScaleMeshSize(Scale Factor=&VectorFromFloat(Value=0.75)) {module=ScaleMeshSize001}
   &ScaleMeshSize(Scale Factor=&VectorFromFloat(Value=&FloatFromCurve(FloatCurve=DataInterface() {propertyValues {Curve {keys [
          <at 0 0 RCIM_Cubic>
          {arriveTangent=1.1765, leaveTangent=1.1765, time=0.1, value=1}
          {interpMode=RCIM_Cubic, tangentMode=RCTM_Break, time=0.8, value=1}
          {arriveTangent=-23.4104, interpMode=RCIM_Cubic, tangentMode=RCTM_Break, time=1, value=0}
         ]}}})))
   &ScaleSpriteSize(Uniform Curve Sprite Scale=DataInterface() {propertyValues {Curve {keys [
        <at 0 0 RCIM_Cubic>
        {arriveTangent=1.1765, leaveTangent=1.1765, time=0.1, value=1}
        {interpMode=RCIM_Cubic, tangentMode=RCTM_Break, time=0.8, value=1}
        {arriveTangent=-23.4104, interpMode=RCIM_Cubic, tangentMode=RCTM_Break, time=1, value=0}
       ]}}})
   &CurlNoiseForce(Bias Noise=false, Noise Frequency=5, Noise Strength=750, Pan Noise=false)
   &WindForce(Wind Speed Scale=&Lerp_Float(A=0.1, B=2.5, Alpha=$NPC.FortniteNPC.WindGustsSystem_Active))
   &AccelerationForce(Acceleration=<50 0 300>)
   &Drag(Drag=10, Use Linear Drag=true, Use Rotational Drag=true)
   &PointAttractionForce(
    Attraction Radius=300
    Attraction Strength=-0.025
    Falloff Exponent=2
    Kill Within Radius=false
    Module.Use Falloff=true
    UseAttractorOffset=false
   )
   /Niagara/Modules/Math/AvoidDistanceFieldSurfaces_GPU(Avoid Nearest Surface=true, Avoid Oncoming Surfaces=true, Nearest Surface Avoidance Strength=50)
   &SolveForcesAndVelocity(Clamp Velocity=false, Limit Acceleration=false)
   &DynamicMaterialParameters(
    Index 0 Param 4=&Lerp_Float(A=10, B=25, Alpha=&Sine(Normalized Angle=$Particles.Age, Period=1, Scale=1, Bias=0.5))
    Index 1 Param 1=1
    Module.Index0Param0Write=true
    Module.Index0Param1Write=true
    Module.Index0Param2Write=true
    Module.Index0Param3Write=true
    Module.Index1Param0Write=true
    Module.Index1Param1Write=true
    Module.Index1Param2Write=true
    Module.Index1Param3Write=true
    Write Parameter Index 1=true
   )
   &ScaleColor(
    Linear Color Curve=DataInterface() {propertyValues {
      AlphaCurve {keys [<at 0.8906 0.45>,<at 1 0>]}
      BlueCurve {keys [<at 0 1>,<at 1 1>]}
      GreenCurve {keys [<at 0 1>,<at 1 1>]}
      RedCurve {keys [<at 0 1>,<at 1 1>]}
     }}
    Scale Mode=::NewEnumerator2
   )
   &InterpolateOverTime(Rate Of Change=0.1, Target Position Value:NiagaraPosition=$Particles.Position, Variable Type=::NewEnumerator6)
   /SetVariables(Particles.Vector 4:Vector4f=&MakeVector4(
     X=&MakeFloatFromVector(
      Vector=&Subtract_Position(A=$Particles.InterpolateOverTime.InterpolateOverTime_Position.Moving Average, B=$Particles.Position)
      Channel=::NewEnumerator0
     )
     Y=&MakeFloatFromVector(
      Vector=&Subtract_Position(A=$Particles.InterpolateOverTime.InterpolateOverTime_Position.Moving Average, B=$Particles.Position)
      Channel=::NewEnumerator1
     )
     Z=&MakeFloatFromVector(
      Vector=&Subtract_Position(A=$Particles.InterpolateOverTime.InterpolateOverTime_Position.Moving Average, B=$Particles.Position)
      Channel=::NewEnumerator2
     )
     W=1
    )) {enabled=false, isSetParameters=true}
   &ScaleColor(Scale Mode=::NewEnumerator2) {enabled=false, module=ScaleColor001}
   &FN_Sprite_Attributes(
    Camera Offset=10
    Color=&HueShiftLinearColor(Hue=0, Saturation=0.65, Value=5, Linear Color=$Particles.Color)
    Create Camera Offset Attribute=true
    Create Color Attribute=true
    Create Position Attribute=true
    Create SpriteSize Attribute=true
    Position:NiagaraPosition=&SubtractVectorFromPosition(Position=$Particles.Position, Vector=<0 0 0>)
    SpriteSize=&Multiply_Vector2D_ByFloat(
     Float=&Lerp_Float(
      A=0
      B=1
      Alpha=&RemapRange(
       Input Value To Remap=&VectorLength(Vector=&SubtractVectorFromPosition(
         Position=/Niagara/DynamicInputs/Helpers/SimulationPosition()
         Vector=Linked() {linkedVariable {name=Particles.Position, type=NiagaraPosition}}
        ))
       Input Min=250
       Input Max=1500
       Output Min=0
       Output Max=1
      )
     )
     Vector2D=<50 50>
    )
    Write to Module Namespace=true
   )
   /SetVariables(Particles.Vector 4:Vector4f=<1 1 0 1>) {isSetParameters=true}
  ]
  rendererClasses [&NiagaraMeshRendererProperties, &NiagaraSpriteRendererProperties]
  renderers [
   {rendererClass=&NiagaraMeshRendererProperties, rendererIndex=0}
   {rendererClass=&NiagaraSpriteRendererProperties, rendererIndex=1}
   {rendererClass=&NiagaraMeshRendererProperties, rendererIndex=2}
   {rendererClass=&NiagaraSpriteRendererProperties, rendererIndex=3}
   {rendererClass=&NiagaraSpriteRendererProperties, rendererIndex=4}
   {rendererClass=&NiagaraSpriteRendererProperties, rendererIndex=5}
  ]
 }
 Local_Fog {
  simTarget=GPUComputeSim
  emitterSpawnScript [/SetVariables(Emitter.Camera Position:NiagaraPosition=<0 0 -35>) {isSetParameters=true, module=SetVariables_#3}]
  emitterUpdateScript [
   &EmitterState()
   &SpawnRate(Spawn Probability=0.5, SpawnRate=50, Use Spawn Probability=true)
  ]
  particleSpawnScript [
   &InitializeParticle(
    Color=<rgba 0.3768 0.2658 0.526 0.15>
    Lifetime Max=6
    Lifetime Min=4
    Lifetime Mode=::NewEnumerator1
    Mesh Scale Mode=::NewEnumerator1
    Mesh Uniform Scale Max=1
    Mesh Uniform Scale Min=0.5
    Position=$Emitter.Camera Position
    Position Mode=::NewEnumerator1
    Sprite Size Max=<700 400>
    Sprite Size Min=<250 250>
    Sprite Size Mode=::NewEnumerator2
    UsePositionOffset=false
    Write Material Random=true
   )
   &ShapeLocation(MaskSpawnGroup=false, Non Uniform Scale=<1 1 0.015>, Override Seed=false, Sphere Radius=6500, UseOwnerScale=true)
   &SampleTexture(
    Apply Sampled RGBA to Particle Color=false
    Comparison Type:ENiagaraBooleanLogicOps_v2=::NewEnumerator6
    "Comparison Value (A)":ENiagara_ImportanceRejectionMode=::NewEnumerator4
    "Comparison Value (B)"=0.5
    Enable Sample Rejection=true
    EnableFalloff=false
    Generate UV Using Importance Sampling=true
    Importance Method:ENiagara_ImportanceColorMode=::NewEnumerator3
    Importance RGBA Channel=1
    Rotation Coordinate Space:ENiagaraCoordinateSpace=::World
    Rotation Mode:ENiagara_RotationMode=::NewEnumerator0
    Texture=DataInterface() {propertyValues {
      Texture=/Game/Effects/Fort_Effects/Effects/Environment/MagicForest/Systems/Textures/T_MagicField_FXMask_Fog
      TextureUserParameter {parameter {name=None, typeDefHandle {registeredTypeIndex=2}, varData []}}
     }}
    Transform UV Into Position=true
    UV Origin:NiagaraPosition=&AddVectorToPosition(Position=$Emitter.Camera Position, Vector=<-500 -8500 500>)
    UV Scale:Vector2f=<63500 63500>
   ) {enabled=false}
   &self:SampleLandscape(Offset:NiagaraPosition=<0 0 1>) {enabled=false, module=SampleLandscape}
   &KillParticlesInVolume(Invert Volume=true, Sphere Radius=3000) {enabled=false}
  ]
  particleUpdateScript [
   &ParticleState()
   &ScaleColor(
    Linear Color Curve=DataInterface() {propertyValues {
      AlphaCurve {keys [<at 0 0>,<at 0.4945 0.5>,<at 1 0>]}
      BlueCurve {keys [<at 0 0>,<at 0.4952 1>,<at 1 0>]}
      GreenCurve {keys [<at 0 0>,<at 0.4952 1>,<at 1 0>]}
      RedCurve {keys [<at 0 0>,<at 0.4952 1>,<at 1 0>]}
     }}
    Scale Mode=::NewEnumerator2
   )
   &WindForce(
    Friction Mode:ENiagaraWindFrictionMode=::NewEnumerator0
    Surface Alignment Mode:ENiagaraWindCombingMode=::NewEnumerator0
    Turbulence Mode:ENiagaraWindTurbulenceMode=::NewEnumerator0
    Wind Speed=<800 0 0>
    Wind Speed Scale=0.85
   )
   &Drag(Use Linear Drag=true, Use Rotational Drag=true)
   &SolveForcesAndVelocity(Clamp Velocity=false, Limit Acceleration=false)
  ]
  rendererClasses [&NiagaraSpriteRendererProperties]
  renderers [
   {rendererClass=&NiagaraSpriteRendererProperties}
   {rendererClass=&NiagaraSpriteRendererProperties, rendererIndex=1}
  ]
 }
}
summary {
 systemName=NS_Magic_AOE_Base
 emitters [
  {
   emitterName=Firefiles_Epic
   simTarget=GPUComputeSim
   rendererClasses [&NiagaraMeshRendererProperties, &NiagaraSpriteRendererProperties]
  }
  {emitterName=Local_Fog, simTarget=GPUComputeSim, rendererClasses [&NiagaraSpriteRendererProperties]}
  {emitterName=Dots_NoTilt, simTarget=GPUComputeSim, rendererClasses [&NiagaraSpriteRendererProperties]}
 ]
 userVariables []
}
